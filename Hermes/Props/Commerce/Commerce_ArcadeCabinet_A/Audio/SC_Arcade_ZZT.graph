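# SoundCue that plays the looping ZZT music track for the Commerce Arcade Cabinet A prop.
# A delay node with a randomized start of up to fifteen seconds feeds a looping wave player, so the music does not begin immediately.
# It uses arcade-machine attenuation, a POI music sound class, and both cabinet-specific and shared arcade concurrency, and restarts with seek when virtualized.
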SoundCue /Hermes/Props/Commerce/Commerce_ArcadeCabinet_A/Audio/SC_Arcade_ZZT
root=Delay_1
nodes [
 {id=Delay_1, children [WavePlayer_0], properties {delayMax=15}}
 {id=WavePlayer_0, properties {bLooping=true, soundWaveAssetPtr=&/MX_Arcade_ZZT_LP_01}}
]
properties {
 attenuationSettings=&/ATT_ArcadeMachines
 maxDistance=3000
 soundClassObject=/Game/Sounds/FortSoundClassesAndModes/Fort_Music_POI_PSM
 virtualizationMode=SeekRestart
 volumeMultiplier=0.25
 concurrencySet [
  &/CONC_ArcadeMachine_ZZT
  &/CONC_ArcadeMachines
 ]
}
