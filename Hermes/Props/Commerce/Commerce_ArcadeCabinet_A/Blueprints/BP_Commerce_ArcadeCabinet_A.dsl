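; Arcade cabinet building prop whose screen or skin material and ambient music loop vary with its building texture data.
; The construction script looks up the first texture data entry in a material map for mesh slot 1, falling back to variant A if unmapped.
; On a material match, an audio map can set the Arcade_Music component's sound loop.
; Exposed-asset-provider overrides register the mapped materials, texture-data override materials and sounds as allowed references.
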(event EventBeginPlay)

(event Collision|EventActorBeginOverlap
  (params (OtherActor :type Actor%20Object%20Reference)))

(event EventTick
  (params (DeltaSeconds :type Float%20%28single-precision%29)))

(fn BuildExposedAssetProviderVersion(Implementation)
  (params (Version :type Fort%20Exposed%20Asset%20Provider%20Version%20Structure%20%28by%20ref%29))
  (|CallParentFunction Version)
  (ExposedAssetProvider|AddVersion Version 2)
  (return))

(fn CollectAssetReferencestoAllow(Implementation)
  (params (Collector :type Fort%20Exposed%20Asset%20Reference%20Collector%20Structure%20%28by%20ref%29))
  (|CallParentFunction Collector)
  (bind _values (Utilities|Map|Values (Variables|Default|GetMaterialMap)))
  (for _array_element _values
    (ExposedAssetProvider|CollectIndirectAssetReference (Variables|GetCollector) (Utilities|ResolveSoftReference _array_element)))
  (bind _keys (Utilities|Map|Keys (Variables|Default|GetMaterialMap)))
  (for _array_element_2 _keys
    (ExposedAssetProvider|CollectIndirectAssetReference (Variables|GetCollector) (Utilities|ResolveSoftReference (Variables|/Script/FortniteGame.BuildingTextureData|GetOverrideMaterial _array_element_2))))
  (bind _values_2 (Utilities|Map|Values (Variables|Default|GetAudioMap)))
  (for _array_element_3 _values_2
    (ExposedAssetProvider|CollectIndirectAssetReference (Variables|GetCollector) (Utilities|ResolveSoftReference _array_element_3)))
  (return))

(fn ConstructionScript
  (bind _audiomap (Variables|Default|GetAudioMap))
  (bind _materialmap (Variables|Default|GetMaterialMap))
  (bind _staticmeshcomponent (Variables|Debug|GetStaticMeshComponent))
  (|CallParentFunction)
  (bind _outtexturedataarray (Building|GetBuildingTextureData))
  (if (> (Utilities|Array|Length _outtexturedataarray) 0)
    (if (.ReturnValue (Utilities|Map|Find _materialmap (Utilities|Array|Get(acopy) _outtexturedataarray)))
      (Rendering|Material|SetMaterial _staticmeshcomponent 1 (Utilities|Map|Find _materialmap (Utilities|Array|Get(acopy) _outtexturedataarray)))
      (if (.ReturnValue (Utilities|Map|Find _audiomap (Utilities|Array|Get(acopy) _outtexturedataarray)))
        (Variables|/Game/Sounds/Fort_Audio_Blueprints/PropAmbientAudioComponent/FortPropAudioComponent.FortPropAudioComponent_C|SetSoundLoop (Utilities|Map|Find _audiomap (Utilities|Array|Get(acopy) _outtexturedataarray)) (Variables|Default|GetArcade_Music)))
      (else
        (Rendering|Material|SetMaterial _staticmeshcomponent 1 "/Hermes/Props/Commerce/Commerce_ArcadeCabinet_A/Materials/MI_Commerce_ArcadeCabinet_A_A.MI_Commerce_ArcadeCabinet_A_A")))))
