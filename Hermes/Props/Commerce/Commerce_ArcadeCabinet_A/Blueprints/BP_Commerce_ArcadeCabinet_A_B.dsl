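; Fortnite arcade cabinet prop Blueprint that picks a location-seeded random look and matching music loop in its construction script.
; When building texture data exists, GenerateRandomSeeds makes two streams from actor location and swatch transition distance; only seed A is used here.
; The chosen swatch's override material fills mesh slot 0, and its index selects the slot 1 cabinet material.
; AudioMap_0 maps that swatch to the Arcade_Music sound loop, updated only when different.
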(event EventBeginPlay)

(event Collision|EventActorBeginOverlap
  (params (OtherActor :type Actor%20Object%20Reference)))

(event EventTick
  (params (DeltaSeconds :type Float%20%28single-precision%29)))

(fn GenerateRandomSeeds
  (Utilities|FlowControl|Sequence
    (:then_0
      (bind _returnvalue (Math|Random|MakeRandomStream (Math|Float|Truncate (* (Class|EnvironmentArtFuncLib|RandomFloatbyLocation (Transformation|GetActorLocation) (Variables|RandomSwatch|GetSwatchTransitionDistance) true) 1.2e+04))))
      (Variables|Default|SetRandomSeedA _returnvalue))
    (:then_1
      (bind _returnvalue_2 (Math|Random|MakeRandomStream (Math|Float|Truncate (* (Class|EnvironmentArtFuncLib|RandomFloatbyLocation (Transformation|GetActorLocation) (Variables|RandomSwatch|GetSwatchTransitionDistance) true) 3e+04))))
      (Variables|Default|SetRandomSeedB _returnvalue_2))))

(fn ConstructionScript
  (bind _audiomap_0 (Variables|Default|GetAudioMap_0))
  (bind _sm_commerce_arcadecabinet_a (Variables|Default|GetSM_Commerce_ArcadeCabinet_A))
  (|CallParentFunction)
  (bind _outtexturedataarray (Building|GetBuildingTextureData))
  (if (> (Utilities|Array|Length _outtexturedataarray) 0)
    (CallFunction|GenerateRandomSeeds)
    (bind _color_swatches (SwatchList|GetSwatchArray))
    (bind (_outitem _outindex) (Utilities|Array|RandomArrayItemfromStream _color_swatches (Variables|Default|GetRandomSeedA)))
    (Utilities|FlowControl|Sequence
      (:then_0
        (Rendering|Material|SetMaterial _sm_commerce_arcadecabinet_a 0 (Variables|/Script/FortniteGame.BuildingTextureData|GetOverrideMaterial _outitem))
        (if (.ReturnValue (Utilities|Map|Find _audiomap_0 _outitem))
          (if (not (== (Variables|/Game/Sounds/Fort_Audio_Blueprints/PropAmbientAudioComponent/FortPropAudioComponent.FortPropAudioComponent_C|GetSoundLoop (Variables|Default|GetArcade_Music)) (Utilities|Map|Find _audiomap_0 _outitem)))
            (Variables|/Game/Sounds/Fort_Audio_Blueprints/PropAmbientAudioComponent/FortPropAudioComponent.FortPropAudioComponent_C|SetSoundLoop (Utilities|Map|Find _audiomap_0 _outitem) (Variables|Default|GetArcade_Music_2)))))
      (:then_1
        (Rendering|Material|SetMaterial _sm_commerce_arcadecabinet_a 1 (Utilities|Array|Get(acopy) (Variables|Default|GetAR_CabinetMaterials) _outindex))))))
