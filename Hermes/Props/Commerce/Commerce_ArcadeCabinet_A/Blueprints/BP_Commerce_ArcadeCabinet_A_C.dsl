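; Arcade cabinet prop that randomizes the look and ambient music of two cabinet meshes.
; When building texture data exists, the construction script derives two random streams from actor location, each picking a color swatch.
; Each swatch sets its mesh's override material and a mapped second material.
; A mapped sound loop goes to that cabinet's audio component, skipping loops that duplicate the base music or the first cabinet's track.
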(event EventBeginPlay)

(event Collision|EventActorBeginOverlap
  (params (OtherActor :type Actor%20Object%20Reference)))

(event EventTick
  (params (DeltaSeconds :type Float%20%28single-precision%29)))

(fn GenerateRandomSeeds
  (Utilities|FlowControl|Sequence
    (:then_0
      (bind _returnvalue (Math|Random|MakeRandomStream (Math|Float|Truncate (* (Class|EnvironmentArtFuncLib|RandomFloatbyLocation (Transformation|GetActorLocation) (Variables|RandomSwatch|GetSwatchTransitionDistance) true) 256841.0))))
      (Variables|Default|SetRandomSeedA _returnvalue))
    (:then_1
      (bind _returnvalue_2 (Math|Random|MakeRandomStream (Math|Float|Truncate (* (Class|EnvironmentArtFuncLib|RandomFloatbyLocation (Transformation|GetActorLocation) (Variables|RandomSwatch|GetSwatchTransitionDistance) true) 486327.0))))
      (Variables|Components|SetRandomSeedB _returnvalue_2))))

(fn GetCreativeActorBounds
  (return false "0, 0, 0"))

(fn GetCreativeActorOrigin
  (return false "0, 0, 0"))

(fn GetCreativeActorRelativeTransform
  (return false 0))

(fn GetGridAxisGuideAttachmentPoint
  (return false false "0, 0, 0"))

(fn ConstructionScript
  (bind _audiomap_0 (Variables|Default|GetAudioMap_0))
  (bind _sm_commerce_arcadecabinet_a (Variables|Default|GetSM_Commerce_ArcadeCabinet_A))
  (bind _sm_commerce_arcadecabinet_a1 (Variables|Default|GetSM_Commerce_ArcadeCabinet_A1))
  (bind _audiomap_0_2 (Variables|Default|GetAudioMap_0))
  (bind _arcade_music (Variables|Default|GetArcade_Music))
  (bind _soundloop (Variables|/Game/Sounds/Fort_Audio_Blueprints/PropAmbientAudioComponent/FortPropAudioComponent.FortPropAudioComponent_C|GetSoundLoop _arcade_music))
  (|CallParentFunction)
  (bind _outtexturedataarray (Building|GetBuildingTextureData))
  (if (> (Utilities|Array|Length _outtexturedataarray) 0)
    (Default|GenerateRandomSeeds)
    (bind _color_swatches (SwatchList|GetSwatchArray))
    (Utilities|FlowControl|Sequence
      (:then_0
        (bind _outitem (Utilities|Array|RandomArrayItemfromStream _color_swatches (Variables|Default|GetRandomSeedA)))
        (Rendering|Material|SetMaterial _sm_commerce_arcadecabinet_a 0 (Variables|/Script/FortniteGame.BuildingTextureData|GetOverrideMaterial _outitem))
        (Rendering|Material|SetMaterial _sm_commerce_arcadecabinet_a 1 (Utilities|Map|Find (Variables|Default|GetMaterialMap) _outitem))
        (if (.ReturnValue (Utilities|Map|Find _audiomap_0_2 _outitem))
          (if (not (== _soundloop (Utilities|Map|Find _audiomap_0_2 _outitem)))
            (Variables|/Game/Sounds/Fort_Audio_Blueprints/PropAmbientAudioComponent/FortPropAudioComponent.FortPropAudioComponent_C|SetSoundLoop (Utilities|Map|Find _audiomap_0_2 _outitem) (Variables|Default|GetArcade_Music_2)))))
      (:then_1
        (bind _outitem_2 (Utilities|Array|RandomArrayItemfromStream _color_swatches (Variables|Components|GetRandomSeedB)))
        (Rendering|Material|SetMaterial _sm_commerce_arcadecabinet_a1 0 (Variables|/Script/FortniteGame.BuildingTextureData|GetOverrideMaterial _outitem_2))
        (Rendering|Material|SetMaterial _sm_commerce_arcadecabinet_a1 1 (Utilities|Map|Find (Variables|Default|GetMaterialMap) _outitem_2))
        (if (.ReturnValue (Utilities|Map|Find _audiomap_0 _outitem_2))
          (if (not (or (== _soundloop (Utilities|Map|Find _audiomap_0 _outitem_2)) (== (Utilities|Map|Find _audiomap_0_2 _outitem) (Utilities|Map|Find _audiomap_0 _outitem_2))))
            (Variables|/Game/Sounds/Fort_Audio_Blueprints/PropAmbientAudioComponent/FortPropAudioComponent.FortPropAudioComponent_C|SetSoundLoop (Utilities|Map|Find _audiomap_0 _outitem_2) (Variables|Default|GetArcade_Music_3))))))))
